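# Niagara effect for the jab of the SaltLattice pickaxe, with a color variant selected by User.Variant.
# One packed emitter fires four burst spawn groups at socket offsets, using the group index as a visibility tag for four mesh renderers and a sprite renderer.
# The last group bursts slightly later as cone-velocity particles, with a platform-dependent count.
# Curves drive scale, fade and material parameters, modulated by time of day and front-end modifiers.
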NiagaraSystem /BRCosmetics/Weapons/FORT_Melee/Pickaxe_SaltLattice/FX/NS_Pickaxe_SaltLattice_Jab
dependencies {
 usedDataInterfaces [
  &NiagaraDataInterfacePlatformSet
  &NiagaraDataInterfaceSpriteRendererInfo
  &NiagaraDataInterfaceCurve
  &NiagaraDataInterfaceColorCurve
 ]
 usedDynamicInputs []
 usedModules [
  &self:SystemSpawnScript.SystemScriptSource.SystemScriptGraph.NiagaraNodeAssignment_0.SetVariables_#1
  &SystemState
 ]
}
emitters {Jab_Packed {
  emitterUpdateScript [
   &EmitterState()
   &SpawnBurst_Instantaneous(Use Spawn Probability=false, UseLoopCountLimit=false)
   &SpawnBurst_Instantaneous(Spawn Group=1, Use Spawn Probability=false, UseLoopCountLimit=false) {module=SpawnBurst_Instantaneous001}
   &SpawnBurst_Instantaneous(Spawn Group=2, Use Spawn Probability=false, UseLoopCountLimit=false) {module=SpawnBurst_Instantaneous002}
   &SpawnBurst_Instantaneous(
    Spawn Count=&MakeIntFromFloat(Float=&FN_PickFloatDependOnPlatform(
      TargetPlatform=DataInterface() {propertyValues {Platforms {qualityLevelMask=-29}}}
      TargetPlatformActive=&MakeFloatFromInt(Int=&RandomRangeInt(Minimum=1, Maximum=2))
      TargetPlatformNoActive=&MakeFloatFromInt(Int=&RandomRangeInt(Minimum=3, Maximum=5))
     ))
    Spawn Group=3
    Spawn Time=0.0333
    Use Spawn Probability=false
    UseLoopCountLimit=false
   ) {module=SpawnBurst_Instantaneous003}
   /SetVariables(
    Emitter.SplashColorCenter:LinearColor=&MergeHelper_ColorSwitch(
     0=<rgba 0.0452 0.3823 0.4 1>
     1=<rgba 0.12 0.386 0.4 1>
     2=<rgba 0.7 0.525 0 1>
     3=<rgba 0.25 0.4708 0.5 1>
     4=<rgba 0.25 0.4708 0.5 1>
     5=<rgba 0.06 0.555 0.6 1>
     Index=&MakeIntFromFloat(Float=$User.Variant)
     Num=::NewEnumerator4
    )
    Emitter.SplashColorOuter:LinearColor=&MergeHelper_ColorSwitch(
     0=<rgba 0.017 0.068 0.17 1>
     1=<rgba 0 0.0875 0.15 1>
     2=<rgba 0.18 0.1002 0.009 1>
     3=<rgba 0.0011 0.1113 0.19 1>
     4=<rgba 0 0.133 0.19 1>
     5=<rgba 0.0281 0.0281 0.225 1>
     Index=&MakeIntFromFloat(Float=$User.Variant)
     Num=::NewEnumerator4
    )
   ) {isSetParameters=true}
  ]
  particleSpawnScript [
   &InitializeParticle(
    Color=<rgba 1 1 1 1.5>
    Lifetime=&MaskFloat(
     Mask Source=1
     Mask Parameter=$Emitter.SpawnGroup
     Mask=2
     Masked Float=0.2
     Passthrough Float=&MaskFloat(
      Mask Source=1
      Mask Parameter=$Emitter.SpawnGroup
      Mask=3
      Masked Float=&RandomRangeFloat(Minimum=0.22, Maximum=0.25)
      Passthrough Float=0.25
     )
    )
    Mesh Scale=&MergeHelper_VectorSwitch(0=<1 1 1>, 1=<1.3 1.3 1.3>, 2=<0.3 0.2 0.2>, 3=?, Index=$Emitter.SpawnGroup, Num=::NewEnumerator1)
    Mesh Scale Mode=::NewEnumerator4
    Sprite Rotation Mode=::NewEnumerator2
    Sprite Size Mode=::NewEnumerator3
    Sprite UV Mode=::NewEnumerator3
    Uniform Sprite Size=&Multiply_Float(
     A=&RandomRangeFloat(Minimum=5.5, Maximum=10)
     B=&FN_FrontEndFloatModifier(Gameplay=1, Front End=0.9)
    )
    UsePositionOffset=false
    Write Material Random=true
   )
   &InitialMeshOrientation(Mesh Orientation Mode=::NewEnumerator2, Rotation=<0 -0.25 0>, Use Rotation Vector=true)
   &AddVelocity(
    Cone Angle=&MergeHelper_FloatSwitch(0=0, 1=0, 2=0, 3=35, Index=$Emitter.SpawnGroup, Num=::NewEnumerator2)
    Cone Axis=<0 0 -1>
    Module.UseConeFalloff=false
    Module.UseVelDistribution=false
    Override Seed=false
    Velocity Mode=::NewEnumerator2
    Velocity Speed=&MergeHelper_FloatSwitch(
     0=100
     1=0
     2=0
     3=&RandomRangeFloat(Minimum=200, Maximum=350)
     Index=$Emitter.SpawnGroup
     Num=::NewEnumerator2
    )
   )
   &SocketLocation(
    Orientation Sampling=::NewEnumerator1
    SamplePosOffset=true
    Sampled Position Offset=&MergeHelper_VectorSwitch(0=<0 7.5 120>, 1=<0 8 130>, 2=<0 8 120>, 3=<0 7.5 100>, Index=$Emitter.SpawnGroup, Num=::NewEnumerator2)
    Socket Reader=$System.Socket Reader
    Socket Sampling Mode=::NewEnumerator3
    VelocitySpeedLimit=false
   )
   &CameraOffset(Camera Offset Amount=&MaskFloat(Mask Source=1, Mask Parameter=$Emitter.SpawnGroup, Mask=2, Masked Float=30, Passthrough Float=40))
   &SubUVAnimation(
    End Frame Range Override=3
    Evaluation Type=::NewEnumerator0
    Output Alternate Renderer Binding=true
    "Sprite Renderer "=DataInterface() {propertyValues {SpriteRenderer=&self:Jab_Packed_0.NiagaraSpriteRendererProperties_0}}
    SubUV Animation Mode:ENiagara_SubUVLookupModeV2=::NewEnumerator2
    UseEndFrame=true
    UseStartFrame=true
   )
   /SetVariables(Particles.VisibilityTag=$Emitter.SpawnGroup) {isSetParameters=true}
  ]
  particleUpdateScript [
   &ParticleState()
   &ScaleSpriteSize(
    Uniform Curve Scale=&FN_FrontEndFloatModifier(Gameplay=1.7, Front End=1)
    Uniform Curve Sprite Scale=DataInterface() {propertyValues {Curve {keys [
        {interpMode=RCIM_Cubic, leaveTangent=6.6646, tangentMode=RCTM_Break, time=0, value=1}
        {arriveTangent=1.8771, interpMode=RCIM_Cubic, leaveTangent=1.8771, tangentMode=RCTM_User, time=0.4, value=2.5}
        {arriveTangent=0.9481, interpMode=RCIM_Cubic, leaveTangent=0.9481, tangentMode=RCTM_User, time=0.9983, value=3.3515}
       ]}}}
   )
   &ScaleMeshSize(Scale Factor=&MergeHelper_VectorSwitch(
     0=&MakeVector(
      X=&FloatFromCurve(FloatCurve=DataInterface() {propertyValues {Curve {keys [<at 0 1>,<at 1 1.5>]}}})
      Y=1
      Z=1
     )
     1=<1 1 1>
     2=&MakeVector(
      X=&FloatFromCurve(FloatCurve=DataInterface() {propertyValues {Curve {keys [{arriveTangent=0.286, interpMode=RCIM_Cubic, leaveTangent=0.286, tangentMode=RCTM_User, time=0, value=1}, <at 1 1.2>]}}})
      Y=&FloatFromCurve(FloatCurve=DataInterface() {propertyValues {Curve {keys [{arriveTangent=3.1084, interpMode=RCIM_Cubic, leaveTangent=3.1084, tangentMode=RCTM_User, time=0, value=0.5}, <at 0.3 1>]}}})
      Z=&FloatFromCurve(FloatCurve=DataInterface() {propertyValues {Curve {keys [{arriveTangent=3.1084, interpMode=RCIM_Cubic, leaveTangent=3.1084, tangentMode=RCTM_User, time=0, value=0.5}, <at 0.3 1>]}}})
     )
     3=?
     Index=$Particles.VisibilityTag
     Num=::NewEnumerator1
    ))
   &SolveForcesAndVelocity(Clamp Velocity=false, Limit Acceleration=false)
   &DynamicMaterialParameters(
    Index 0 Param 1=&FloatFromCurve(FloatCurve=DataInterface() {propertyValues {Curve {keys [<at 0 0>,<at 0.5 0>,<at 1 0.08>]}}})
    Index 0 Param 2=&FloatFromCurve(FloatCurve=DataInterface() {propertyValues {Curve {keys [{arriveTangent=-2.2616, interpMode=RCIM_Cubic, time=0, value=-0.5}, {interpMode=RCIM_Cubic, leaveTangent=-0.7156, time=1, value=1}]}}})
    Index 0 Param 3=1.2
    Index 0 Param 4=&FloatFromCurve(FloatCurve=DataInterface() {propertyValues {Curve {keys [<at 0 1>,<at 1 2>]}}}, Scale Curve=0.2)
    Module.Index0Float0Write=true
    Module.Index0Float1Write=true
    Module.Index0Float2Write=true
    Module.Index0Float3Write=true
   )
   &ScaleColor(
    Scale Alpha=&FloatFromCurve(
     FloatCurve=DataInterface() {propertyValues {Curve {keys [{tangentMode=RCTM_User, time=0, value=1}, {tangentMode=RCTM_User, time=1, value=0}]}}}
     Scale Curve=&Multiply_Float(
      A=&MergeHelper_FloatSwitch(
       0=0.5
       1=0.5
       2=0.5
       3=0.5
       4=5
       5=0.5
       Index=&MakeIntFromFloat(Float=$User.Variant)
       Num=::NewEnumerator4
      )
      B=$System.ToDScalar
     )
    )
    Scale RGB=&MakeVectorFromLinearColorRGB(Color=&MergeHelper_ColorSwitch(
      0=&ColorFromCurve(ColorCurve=DataInterface() {propertyValues {
         AlphaCurve {keys []}
         BlueCurve {keys [<at 0 20 RCIM_Cubic>,<at 0.5024 0.2 RCIM_Cubic>]}
         GreenCurve {keys [<at 0 6.6667 RCIM_Cubic>,<at 0.5024 0.0333 RCIM_Cubic>]}
         RedCurve {keys [<at 0 0 RCIM_Cubic>,<at 0.5024 0 RCIM_Cubic>]}
        }})
      1=&ColorFromCurve(ColorCurve=DataInterface() {propertyValues {
         AlphaCurve {keys []}
         BlueCurve {keys [<at 0 20 RCIM_Cubic>,<at 0.5024 0.2 RCIM_Cubic>]}
         GreenCurve {keys [<at 0 14.6667 RCIM_Cubic>,<at 0.5024 0.1167 RCIM_Cubic>]}
         RedCurve {keys [<at 0 0 RCIM_Cubic>,<at 0.5024 0 RCIM_Cubic>]}
        }})
      2=&ColorFromCurve(ColorCurve=DataInterface() {propertyValues {
         AlphaCurve {keys []}
         BlueCurve {keys [<at 0 1 RCIM_Cubic>,<at 0.5024 0 RCIM_Cubic>]}
         GreenCurve {keys [<at 0 11.1333 RCIM_Cubic>,<at 0.5024 0.0733 RCIM_Cubic>]}
         RedCurve {keys [<at 0 20 RCIM_Cubic>,<at 0.5024 0.2 RCIM_Cubic>]}
        }})
      3=&ColorFromCurve(ColorCurve=DataInterface() {propertyValues {
         AlphaCurve {keys []}
         BlueCurve {keys [<at 0 20 RCIM_Cubic>,<at 0.5024 0.2 RCIM_Cubic>]}
         GreenCurve {keys [<at 0 16 RCIM_Cubic>,<at 0.5024 0.15 RCIM_Cubic>]}
         RedCurve {keys [<at 0 0 RCIM_Cubic>,<at 0.5024 0 RCIM_Cubic>]}
        }})
      4=&ColorFromCurve(ColorCurve=DataInterface() {propertyValues {
         AlphaCurve {keys []}
         BlueCurve {keys [<at 0 5 RCIM_Cubic>,<at 0.7529 0.2>]}
         GreenCurve {keys [<at 0 0.25 RCIM_Cubic>,<at 0.7529 0.0167>]}
         RedCurve {keys [<at 0 1.0417 RCIM_Cubic>,<at 0.7529 0>]}
        }})
      5=&ColorFromCurve(ColorCurve=DataInterface() {propertyValues {
         AlphaCurve {keys []}
         BlueCurve {keys [<at 0 20 RCIM_Cubic>,<at 0.5024 0.2 RCIM_Cubic>]}
         GreenCurve {keys [<at 0 16 RCIM_Cubic>,<at 0.5024 0.15 RCIM_Cubic>]}
         RedCurve {keys [<at 0 0 RCIM_Cubic>,<at 0.5024 0 RCIM_Cubic>]}
        }})
      Index=&MakeIntFromFloat(Float=$User.Variant)
      Num=::NewEnumerator4
     ))
    ScaleA=true
    ScaleRGB=true
   )
   &FN_Mesh_Attributes(
    Color=&MakeLinearColorFromVectorAndFloat(
     "Vector (RGB)"=&MakeVectorFromLinearColorRGB(Color=&MergeHelper_ColorSwitch(
       0=&ColorFromCurve(ColorCurve=DataInterface() {propertyValues {
          AlphaCurve {keys []}
          BlueCurve {keys [<at 0 5 RCIM_Cubic>,<at 0.7529 0.2>]}
          GreenCurve {keys [<at 0 0.8333 RCIM_Cubic>,<at 0.7529 0.0333>]}
          RedCurve {keys [<at 0 0 RCIM_Cubic>,<at 0.7529 0>]}
         }})
       1=&ColorFromCurve(ColorCurve=DataInterface() {propertyValues {
          AlphaCurve {keys []}
          BlueCurve {keys [<at 0 5 RCIM_Cubic>,<at 0.7529 0.2>]}
          GreenCurve {keys [<at 0 4.1667 RCIM_Cubic>,<at 0.7529 0.1067>]}
          RedCurve {keys [<at 0 0 RCIM_Cubic>,<at 0.7529 0>]}
         }})
       2=&ColorFromCurve(ColorCurve=DataInterface() {propertyValues {
          AlphaCurve {keys []}
          BlueCurve {keys [<at 0 0.25 RCIM_Cubic>,<at 0.7529 0>]}
          GreenCurve {keys [<at 0 2.7833 RCIM_Cubic>,<at 0.7529 0.0733>]}
          RedCurve {keys [<at 0 5 RCIM_Cubic>,<at 0.7529 0.2>]}
         }})
       3=&ColorFromCurve(ColorCurve=DataInterface() {propertyValues {
          AlphaCurve {keys []}
          BlueCurve {keys [<at 0 5 RCIM_Cubic>,<at 0.7529 0.2>]}
          GreenCurve {keys [<at 0 4.2083 RCIM_Cubic>,<at 0.7529 0.15>]}
          RedCurve {keys [<at 0 0.25 RCIM_Cubic>,<at 0.7529 0>]}
         }})
       4=&ColorFromCurve(ColorCurve=DataInterface() {propertyValues {
          AlphaCurve {keys []}
          BlueCurve {keys [<at 0 5 RCIM_Cubic>,<at 0.7529 0.2>]}
          GreenCurve {keys [<at 0 0.25 RCIM_Cubic>,<at 0.7529 0.0167>]}
          RedCurve {keys [<at 0 1.0417 RCIM_Cubic>,<at 0.7529 0>]}
         }})
       5=&ColorFromCurve(ColorCurve=DataInterface() {propertyValues {
          AlphaCurve {keys []}
          BlueCurve {keys [<at 0 5 RCIM_Cubic>,<at 0.7529 0.2>]}
          GreenCurve {keys [<at 0 4.2083 RCIM_Cubic>,<at 0.7529 0.15>]}
          RedCurve {keys [<at 0 0.25 RCIM_Cubic>,<at 0.7529 0>]}
         }})
       Index=&MakeIntFromFloat(Float=$User.Variant)
       Num=::NewEnumerator4
      ))
     "Float (Alpha)"=&FloatFromCurve(
      FloatCurve=DataInterface() {propertyValues {Curve {keys [<at 0 1>,<at 1 0>]}}}
      Scale Curve=&Multiply_Float(A=0.5, B=$System.ToDScalar)
     )
    )
    Create Color Attribute=true
    Write to Module Namespace=true
   )
   /SetVariables(Particles.DynamicParametersAlt:Vector4f=&MakeVector4(
     X=&FloatFromCurve(FloatCurve=DataInterface() {propertyValues {Curve {keys [<at 0 0.4>,<at 1 0.8>]}}})
     Y=&FloatFromCurve(FloatCurve=DataInterface() {propertyValues {Curve {keys [
          {arriveTangent=-2.2616, interpMode=RCIM_Cubic, leaveTangent=-3.554, tangentMode=RCTM_Break, time=0, value=1}
          {arriveTangent=-0.8527, interpMode=RCIM_Cubic, leaveTangent=-0.8527, tangentMode=RCTM_User, time=0.8, value=0}
         ]}}})
     Z=0.4
     W=0
    )) {isSetParameters=true}
   &FN_Mesh_Attributes(
    Color=&MultiplyLinearColor(
     Linear Color=<rgba 1 1 1 1.5>
     Scale Factor=&Multiply_Vector4(
      A=&MakeVector4FromVectorAndFloat(
       "Vector (RGB)"=&MakeVectorFromLinearColorRGB(Color=&MergeHelper_ColorSwitch(
         0=&ColorFromCurve(ColorCurve=DataInterface() {propertyValues {
            AlphaCurve {keys []}
            BlueCurve {keys [<at 0 2 RCIM_Cubic>,<at 0.7508 5 RCIM_Cubic>]}
            GreenCurve {keys [<at 0 1.1667 RCIM_Cubic>,<at 0.7508 0.8333 RCIM_Cubic>]}
            RedCurve {keys [<at 0 0 RCIM_Cubic>,<at 0.7508 0 RCIM_Cubic>]}
           }})
         1=&ColorFromCurve(ColorCurve=DataInterface() {propertyValues {
            AlphaCurve {keys []}
            BlueCurve {keys [<at 0 2 RCIM_Cubic>,<at 0.7508 5 RCIM_Cubic>]}
            GreenCurve {keys [<at 0 1.6667 RCIM_Cubic>,<at 0.7508 3.3333 RCIM_Cubic>]}
            RedCurve {keys [<at 0 0 RCIM_Cubic>,<at 0.7508 0 RCIM_Cubic>]}
           }})
         2=&ColorFromCurve(ColorCurve=DataInterface() {propertyValues {
            AlphaCurve {keys []}
            BlueCurve {keys [<at 0 0 RCIM_Cubic>,<at 0.7508 0 RCIM_Cubic>]}
            GreenCurve {keys [<at 0 1.3333 RCIM_Cubic>,<at 0.7508 2.5 RCIM_Cubic>]}
            RedCurve {keys [<at 0 2 RCIM_Cubic>,<at 0.7508 5 RCIM_Cubic>]}
           }})
         3=&ColorFromCurve(ColorCurve=DataInterface() {propertyValues {
            AlphaCurve {keys []}
            BlueCurve {keys [<at 0 2 RCIM_Cubic>,<at 0.7508 5 RCIM_Cubic>]}
            GreenCurve {keys [<at 0 1.6667 RCIM_Cubic>,<at 0.7508 3.75 RCIM_Cubic>]}
            RedCurve {keys [<at 0 0 RCIM_Cubic>,<at 0.7508 0 RCIM_Cubic>]}
           }})
         4=&ColorFromCurve(ColorCurve=DataInterface() {propertyValues {
            AlphaCurve {keys []}
            BlueCurve {keys [<at 0 2 RCIM_Cubic>,<at 0.7508 5 RCIM_Cubic>]}
            GreenCurve {keys [<at 0 1.6667 RCIM_Cubic>,<at 0.7508 1.6667 RCIM_Cubic>]}
            RedCurve {keys [<at 0 0 RCIM_Cubic>,<at 0.7508 0 RCIM_Cubic>]}
           }})
         5=&ColorFromCurve(ColorCurve=DataInterface() {propertyValues {
            AlphaCurve {keys []}
            BlueCurve {keys [<at 0 2 RCIM_Cubic>,<at 0.7508 5 RCIM_Cubic>]}
            GreenCurve {keys [<at 0 1.6667 RCIM_Cubic>,<at 0.7508 3.75 RCIM_Cubic>]}
            RedCurve {keys [<at 0 0 RCIM_Cubic>,<at 0.7508 0 RCIM_Cubic>]}
           }})
         Index=&MakeIntFromFloat(Float=$User.Variant)
         Num=::NewEnumerator4
        ))
       "Float (Alpha)"=&FloatFromCurve(
        FloatCurve=DataInterface() {propertyValues {Curve {keys [<at 0 1>,<at 1 0>]}}}
        Scale Curve=&Multiply_Float(
         A=&MergeHelper_FloatSwitch(
          0=0.25
          1=0.25
          2=0.25
          3=0.25
          4=0.25
          5=0.25
          Index=&MakeIntFromFloat(Float=$User.Variant)
          Num=::NewEnumerator4
         )
         B=$System.ToDScalar
        )
       )
      )
      B=&MakeVector4_FromLinearColor(LinearColor=&MergeHelper_ColorSwitch(
        0=<rgba 1 1 1 1>
        1=<rgba 1 1 1 1>
        2=&ColorFromCurve(
         ColorCurve=DataInterface() {propertyValues {
           AlphaCurve {keys [<at 0 1>,<at 0.125 1>,<at 0.5 1>,<at 0.875 1>,<at 1 1>]}
           BlueCurve {keys [<at 0.0417 0.8469>,<at 0.125 0.8304>,<at 0.25 0.7>,<at 0.375 0.935>,<at 0.5 0.895>,<at 0.625 0.935>,<at 0.75 0.7>,<at 0.875 0.675>,<at 0.9167 1>,<at 1 0.8469>]}
           GreenCurve {keys [<at 0.0417 0.686>,<at 0.125 0.6609>,<at 0.25 0.731>,<at 0.375 0.935>,<at 0.5 0.895>,<at 0.625 0.935>,<at 0.75 0.7>,<at 0.875 0.7467>,<at 0.9167 0.6376>,<at 1 0.686>]}
           RedCurve {keys [<at 0.0417 0.4234>,<at 0.125 1>,<at 0.25 1>,<at 0.375 0.935>,<at 0.5 0.895>,<at 0.625 0.935>,<at 0.75 1>,<at 0.875 1>,<at 0.9167 0.6>,<at 1 0.4234>]}
          }}
         CurveIndex=Linked() {linkedVariable {name=NPC.FortniteNPC.TimeOfDay_0-1, type=NiagaraFloat}}
        )
        3=<rgba 1 1 1 1>
        4=&ColorFromCurve(
         ColorCurve=DataInterface() {propertyValues {
           AlphaCurve {keys [<at 0 1>,<at 0.125 1>,<at 0.5 1>,<at 0.875 1>,<at 1 1>]}
           BlueCurve {keys [<at 0.0417 0.8469>,<at 0.125 0.8304>,<at 0.25 0.7>,<at 0.375 0.935>,<at 0.5 0.895>,<at 0.625 0.935>,<at 0.75 0.7>,<at 0.875 0.675>,<at 0.9167 1>,<at 1 0.8469>]}
           GreenCurve {keys [<at 0.0417 0.686>,<at 0.125 0.6609>,<at 0.25 0.731>,<at 0.375 0.935>,<at 0.5 0.895>,<at 0.625 0.935>,<at 0.75 0.7>,<at 0.875 0.7467>,<at 0.9167 0.6376>,<at 1 0.686>]}
           RedCurve {keys [<at 0.0417 0.4234>,<at 0.125 1>,<at 0.25 1>,<at 0.375 0.935>,<at 0.5 0.895>,<at 0.625 0.935>,<at 0.75 1>,<at 0.875 1>,<at 0.9167 0.6>,<at 1 0.4234>]}
          }}
         CurveIndex=Linked() {linkedVariable {name=NPC.FortniteNPC.TimeOfDay_0-1, type=NiagaraFloat}}
        )
        5=&ColorFromCurve(
         ColorCurve=DataInterface() {propertyValues {
           AlphaCurve {keys [<at 0 1>,<at 0.125 1>,<at 0.5 1>,<at 0.875 1>,<at 1 1>]}
           BlueCurve {keys [<at 0.0417 0.8469>,<at 0.125 0.8304>,<at 0.25 0.7>,<at 0.375 0.935>,<at 0.5 0.895>,<at 0.625 0.935>,<at 0.75 0.7>,<at 0.875 0.675>,<at 0.9167 1>,<at 1 0.8469>]}
           GreenCurve {keys [<at 0.0417 0.686>,<at 0.125 0.6609>,<at 0.25 0.731>,<at 0.375 0.935>,<at 0.5 0.895>,<at 0.625 0.935>,<at 0.75 0.7>,<at 0.875 0.7467>,<at 0.9167 0.6376>,<at 1 0.686>]}
           RedCurve {keys [<at 0.0417 0.4234>,<at 0.125 1>,<at 0.25 1>,<at 0.375 0.935>,<at 0.5 0.895>,<at 0.625 0.935>,<at 0.75 1>,<at 0.875 1>,<at 0.9167 0.6>,<at 1 0.4234>]}
          }}
         CurveIndex=Linked() {linkedVariable {name=NPC.FortniteNPC.TimeOfDay_0-1, type=NiagaraFloat}}
        )
        Index=&MakeIntFromFloat(Float=$User.Variant)
        Num=::NewEnumerator4
       ))
     )
    )
    Create Color Attribute=true
    Write to Module Namespace=true
   ) {module=FN_Mesh_Attributes002}
   &FN_Mesh_Attributes(
    Color=&MultiplyLinearColor(
     Linear Color=&MakeLinearColorFromVectorAndFloat(
      "Vector (RGB)"=&MakeVectorFromLinearColorRGB(Color=&MergeHelper_ColorSwitch(
        0=&ColorFromCurve(ColorCurve=DataInterface() {propertyValues {
           AlphaCurve {keys []}
           BlueCurve {keys [<at 0 2 RCIM_Cubic>,<at 0.7508 5 RCIM_Cubic>]}
           GreenCurve {keys [<at 0 1.1667 RCIM_Cubic>,<at 0.7508 1.6667 RCIM_Cubic>]}
           RedCurve {keys [<at 0 0 RCIM_Cubic>,<at 0.7508 0 RCIM_Cubic>]}
          }})
        1=&ColorFromCurve(ColorCurve=DataInterface() {propertyValues {
           AlphaCurve {keys []}
           BlueCurve {keys [<at 0 2 RCIM_Cubic>,<at 0.7508 5 RCIM_Cubic>]}
           GreenCurve {keys [<at 0 1.6667 RCIM_Cubic>,<at 0.7508 3.3333 RCIM_Cubic>]}
           RedCurve {keys [<at 0 0 RCIM_Cubic>,<at 0.7508 0 RCIM_Cubic>]}
          }})
        2=&ColorFromCurve(ColorCurve=DataInterface() {propertyValues {
           AlphaCurve {keys []}
           BlueCurve {keys [<at 0 0 RCIM_Cubic>,<at 0.7508 0 RCIM_Cubic>]}
           GreenCurve {keys [<at 0 1.3333 RCIM_Cubic>,<at 0.7508 2.5 RCIM_Cubic>]}
           RedCurve {keys [<at 0 2 RCIM_Cubic>,<at 0.7508 5 RCIM_Cubic>]}
          }})
        3=&ColorFromCurve(ColorCurve=DataInterface() {propertyValues {
           AlphaCurve {keys []}
           BlueCurve {keys [<at 0 2 RCIM_Cubic>,<at 0.7508 5 RCIM_Cubic>]}
           GreenCurve {keys [<at 0 1.6667 RCIM_Cubic>,<at 0.7508 3.75 RCIM_Cubic>]}
           RedCurve {keys [<at 0 0 RCIM_Cubic>,<at 0.7508 0 RCIM_Cubic>]}
          }})
        4=&ColorFromCurve(ColorCurve=DataInterface() {propertyValues {
           AlphaCurve {keys []}
           BlueCurve {keys [<at 0 2 RCIM_Cubic>,<at 0.7508 5 RCIM_Cubic>]}
           GreenCurve {keys [<at 0 1.6667 RCIM_Cubic>,<at 0.7508 1.6667 RCIM_Cubic>]}
           RedCurve {keys [<at 0 0 RCIM_Cubic>,<at 0.7508 0 RCIM_Cubic>]}
          }})
        5=&ColorFromCurve(ColorCurve=DataInterface() {propertyValues {
           AlphaCurve {keys []}
           BlueCurve {keys [<at 0 2 RCIM_Cubic>,<at 0.7508 5 RCIM_Cubic>]}
           GreenCurve {keys [<at 0 1.6667 RCIM_Cubic>,<at 0.7508 3.75 RCIM_Cubic>]}
           RedCurve {keys [<at 0 0 RCIM_Cubic>,<at 0.7508 0 RCIM_Cubic>]}
          }})
        Index=&MakeIntFromFloat(Float=$User.Variant)
        Num=::NewEnumerator4
       ))
      "Float (Alpha)"=&FloatFromCurve(
       FloatCurve=DataInterface() {propertyValues {Curve {keys [{tangentMode=RCTM_User, time=0, value=1}, {tangentMode=RCTM_User, time=1, value=0}]}}}
       Scale Curve=0.35
      )
     )
     Scale Factor=&MakeVector4_FromLinearColor(LinearColor=&MergeHelper_ColorSwitch(
       0=<rgba 1 1 1 1>
       1=<rgba 1 1 1 1>
       2=&ColorFromCurve(
        ColorCurve=DataInterface() {propertyValues {
          AlphaCurve {keys [<at 0 1>,<at 0.125 1>,<at 0.5 1>,<at 0.875 1>,<at 1 1>]}
          BlueCurve {keys [<at 0.0417 0.8469>,<at 0.125 0.8304>,<at 0.25 0.7>,<at 0.375 0.935>,<at 0.5 0.895>,<at 0.625 0.935>,<at 0.75 0.7>,<at 0.875 0.675>,<at 0.9167 1>,<at 1 0.8469>]}
          GreenCurve {keys [<at 0.0417 0.686>,<at 0.125 0.6609>,<at 0.25 0.731>,<at 0.375 0.935>,<at 0.5 0.895>,<at 0.625 0.935>,<at 0.75 0.7>,<at 0.875 0.7467>,<at 0.9167 0.6376>,<at 1 0.686>]}
          RedCurve {keys [<at 0.0417 0.4234>,<at 0.125 1>,<at 0.25 1>,<at 0.375 0.935>,<at 0.5 0.895>,<at 0.625 0.935>,<at 0.75 1>,<at 0.875 1>,<at 0.9167 0.6>,<at 1 0.4234>]}
         }}
        CurveIndex=Linked() {linkedVariable {name=NPC.FortniteNPC.TimeOfDay_0-1, type=NiagaraFloat}}
       )
       3=<rgba 1 1 1 1>
       4=&ColorFromCurve(
        ColorCurve=DataInterface() {propertyValues {
          AlphaCurve {keys [<at 0 1>,<at 0.125 1>,<at 0.5 1>,<at 0.875 1>,<at 1 1>]}
          BlueCurve {keys [<at 0.0417 0.8469>,<at 0.125 0.8304>,<at 0.25 0.7>,<at 0.375 0.935>,<at 0.5 0.895>,<at 0.625 0.935>,<at 0.75 0.7>,<at 0.875 0.675>,<at 0.9167 1>,<at 1 0.8469>]}
          GreenCurve {keys [<at 0.0417 0.686>,<at 0.125 0.6609>,<at 0.25 0.731>,<at 0.375 0.935>,<at 0.5 0.895>,<at 0.625 0.935>,<at 0.75 0.7>,<at 0.875 0.7467>,<at 0.9167 0.6376>,<at 1 0.686>]}
          RedCurve {keys [<at 0.0417 0.4234>,<at 0.125 1>,<at 0.25 1>,<at 0.375 0.935>,<at 0.5 0.895>,<at 0.625 0.935>,<at 0.75 1>,<at 0.875 1>,<at 0.9167 0.6>,<at 1 0.4234>]}
         }}
        CurveIndex=Linked() {linkedVariable {name=NPC.FortniteNPC.TimeOfDay_0-1, type=NiagaraFloat}}
       )
       5=&ColorFromCurve(
        ColorCurve=DataInterface() {propertyValues {
          AlphaCurve {keys [<at 0 1>,<at 0.125 1>,<at 0.5 1>,<at 0.875 1>,<at 1 1>]}
          BlueCurve {keys [<at 0.0417 0.8469>,<at 0.125 0.8304>,<at 0.25 0.7>,<at 0.375 0.935>,<at 0.5 0.895>,<at 0.625 0.935>,<at 0.75 0.7>,<at 0.875 0.675>,<at 0.9167 1>,<at 1 0.8469>]}
          GreenCurve {keys [<at 0.0417 0.686>,<at 0.125 0.6609>,<at 0.25 0.731>,<at 0.375 0.935>,<at 0.5 0.895>,<at 0.625 0.935>,<at 0.75 0.7>,<at 0.875 0.7467>,<at 0.9167 0.6376>,<at 1 0.686>]}
          RedCurve {keys [<at 0.0417 0.4234>,<at 0.125 1>,<at 0.25 1>,<at 0.375 0.935>,<at 0.5 0.895>,<at 0.625 0.935>,<at 0.75 1>,<at 0.875 1>,<at 0.9167 0.6>,<at 1 0.4234>]}
         }}
        CurveIndex=Linked() {linkedVariable {name=NPC.FortniteNPC.TimeOfDay_0-1, type=NiagaraFloat}}
       )
       Index=&MakeIntFromFloat(Float=$User.Variant)
       Num=::NewEnumerator4
      ))
    )
    Create Color Attribute=true
    Create Scale Attribute=true
    Scale=&Multiply_VectorByFloat(
     Vector=&Multiply_Vector(
      A=&RandomRangeVector(Minimum=<0.12 0.3 0.3>, Maximum=<0.14 0.3 0.3>)
      B=&MakeVector(
       X=&FloatFromCurve(FloatCurve=DataInterface() {propertyValues {Curve {keys [{arriveTangent=0.3763, interpMode=RCIM_Cubic, leaveTangent=0.3763, tangentMode=RCTM_User, time=0, value=1}, <at 1 1.2>]}}})
       Y=&FloatFromCurve(FloatCurve=DataInterface() {propertyValues {Curve {keys [{arriveTangent=3.1084, interpMode=RCIM_Cubic, leaveTangent=3.1084, tangentMode=RCTM_User, time=0, value=0.5}, <at 0.3 1>]}}})
       Z=&FloatFromCurve(FloatCurve=DataInterface() {propertyValues {Curve {keys [{arriveTangent=3.1084, interpMode=RCIM_Cubic, leaveTangent=3.1084, tangentMode=RCTM_User, time=0, value=0.5}, <at 0.3 1>]}}})
      )
     )
     Float=0.6
    )
    Write to Module Namespace=true
   ) {module=FN_Mesh_Attributes001}
   /SetVariables(Particles.Dy3:Vector4f=&MakeVector4(X=0, Y=1, Z=0, W=1)) {isSetParameters=true}
   &FN_Sprite_Attributes(
    Camera Offset=20
    Color=&MultiplyLinearColor(
     Linear Color=<rgba 2 2 2 1>
     Scale Factor=&Multiply_Vector4(
      A=&MakeVector4FromVectorAndFloat(
       "Vector (RGB)"=&MakeVectorFromLinearColorRGB(Color=&Multiply_LinearColorByFloat(Linear Color=<rgba 1 1 1 1>, Float=&FN_FrontEndFloatModifier(Gameplay=1, Front End=1.3)))
       "Float (Alpha)"=&FloatFromCurve(
        FloatCurve=DataInterface() {propertyValues {Curve {keys [{interpMode=RCIM_Cubic, tangentMode=RCTM_User, time=0.5, value=1.1}, {interpMode=RCIM_Cubic, tangentMode=RCTM_User, time=1, value=0}]}}}
        Scale Curve=&FNTimeOfDayScalar(Night Scale=0.775, Morning Scale=0.55, Evening Scale=0.835)
       )
      )
      B=&MakeVector4_FromLinearColor(LinearColor=&MergeHelper_ColorSwitch(
        0=<rgba 1 1 1 1>
        1=<rgba 1 1 1 1>
        2=<rgba 1 1 1 1>
        3=<rgba 1 1 1 1>
        4=&FNTimeOfDayColor(Night Color=<rgba 1 1 1 1.2>, Morning Color=<rgba 1 1 1 0.6>, Day Color=<rgba 1 1 1 1>, Evening Color=<rgba 1 1 1 1>)
        5=&FNTimeOfDayColor(Night Color=<rgba 1 1 1 1>, Morning Color=<rgba 1 1 1 0.6>, Day Color=<rgba 1 1 1 1>, Evening Color=<rgba 1 1 1 1>)
        Index=&MakeIntFromFloat(Float=$User.Variant)
        Num=::NewEnumerator4
       ))
     )
    )
    Create Camera Offset Attribute=true
    Create Color Attribute=true
    Write to Module Namespace=true
   )
   /SetVariables(Particles.DynamicParametersAlt2:Vector4f=&MakeVector4(
     X=0
     Y=&FloatFromCurve(FloatCurve=DataInterface() {propertyValues {Curve {keys [<at 0.35 0.6 RCIM_Cubic>,<at 1 0.7 RCIM_Cubic>]}}})
     Z=&FloatFromCurve(FloatCurve=DataInterface() {propertyValues {Curve {keys [<at 0.2 -0.5>,<at 0.9 -1>]}}})
     W=&FloatFromCurve(FloatCurve=DataInterface() {propertyValues {Curve {keys [{arriveTangent=2.6651, interpMode=RCIM_Cubic, leaveTangent=2.6651, tangentMode=RCTM_User, time=0, value=0.1}, <at 0.6 1.333 RCIM_Cubic>]}}})
    )) {isSetParameters=true}
  ]
  rendererClasses [&NiagaraMeshRendererProperties, &NiagaraSpriteRendererProperties]
  renderers [
   {rendererClass=&NiagaraMeshRendererProperties, rendererIndex=0}
   {rendererClass=&NiagaraMeshRendererProperties, rendererIndex=1}
   {rendererClass=&NiagaraMeshRendererProperties, rendererIndex=2}
   {rendererClass=&NiagaraMeshRendererProperties, rendererIndex=3}
   {rendererClass=&NiagaraSpriteRendererProperties, rendererIndex=4}
  ]
 }}
summary {
 systemName=NS_Pickaxe_SaltLattice_Jab
 emitters [{emitterName=Jab_Packed, rendererClasses [&NiagaraMeshRendererProperties, &NiagaraSpriteRendererProperties]}]
 userVariables [
  User.bIsFrontEnd {type=NiagaraBool}
  User.bIsFrontEndPreview {type=NiagaraBool}
  User.SwingSectionIndex {type=NiagaraInt32}
  User.Variant {type=NiagaraFloat, defaultValue=0, description=""}
 ]
}
